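# Niagara jetpack exhaust effect for the HighTower Radish pet, with flames, glow and smoke at two skeletal mesh sockets.
# Two persistent flame particles follow the sockets, each driving three mesh layers with curve-pulsed length and a flickering light.
# A delayed smoke trail and two initial-burst smoke emitters use SubUV sprites, pushed along socket-relative offsets as they age.
# Spawn group selects the socket, and System.VehicleAlpha scales alpha in every emitter.
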NiagaraSystem /Game/Effects/Fort_Effects/Effects/Pets/HighTower_Radish/NS_Pet_HighTower_Radish_Jets
dependencies {
 usedDataInterfaces [&NiagaraDataInterfaceCurve]
 usedDynamicInputs [&MakeCustomFloatFromBool]
 usedModules [
  &SystemState
  &self:SystemSpawnScript.SystemScriptSource.SystemScriptGraph.NiagaraNodeAssignment_0.SetVariables_#1
 ]
}
emitters {
 JetPackFlames {
  emitterSpawnScript [&FN_PreserveSkydiveUserVariables() {module=PreserveUserVariables}]
  emitterUpdateScript [
   &EmitterState(
    Life Cycle Mode=::NewEnumerator1
    Loop Behavior=::NewEnumerator1
    Loop Duration=&UniformRangedFloat(Minimum=0.75, Maximum=1.25)
    UseLoopDelay=false
   )
   &SpawnBurst_Instantaneous(Spawn Count=2, Spawn Probability=HlslExpression() {hlslExpression="Emitter.LoopCount==0"}, Use Spawn Probability=true)
  ]
  particleSpawnScript [
   &FN_SetSpawnGroup_v2(Number of Spawn Groups=2)
   /Niagara/Modules/Spawn/Initialization/InitializeParticle(
    Module.Write Color=true
    Module.Write Lifetime=true
    Module.Write Mass=false
    Module.Write Material Random=true
    Module.Write Position=false
    Module.Write Scale=false
    Module.Write SpriteRotation=false
    Module.Write SpriteSize=true
    Sprite Size=<2 2>
   )
  ]
  particleUpdateScript [
   &ParticleState(Kill Particles When Lifetime Has Elapsed=false)
   &SkeletalMeshLocation(
    Apply Sampled Orientation to Mesh Particles=true
    Bone / Socket Index=$Particles.SpawnGroup
    Inherit Velocity=false
    MaskSpawnGroup=false
    Mesh Sampling Type=::NewEnumerator5
    Sampled Orientation Coordinate Space=::World
    Sampled Position Coordinate Space=::World
    Sampled Velocity Coordinate Space:ENiagaraCoordinateSpace=::World
    Skeletal Mesh=$User.SkeletalMesh
    Socket Sampling Mode=::NewEnumerator5
   )
   /Game/Effects/Niagara/Modules/Mesh/FN_MeshRenderer_v3/FN_MeshRenderer_v2(
    Color:LinearColor=<rgba 6 2.6395 0 1>
    Create COlor Attribute=true
    Create Orientation Attribute=true
    Create Scale Attribute=true
    Orientation:Quat4f=&MultiplyQuaternion(
     Quaternion A=$Particles.MeshOrientation
     Quaternion B=&FindQuatBetweenTwoVectors(From Vector=<-1 0 0>, To Vector=<0 0 -1>)
    )
    Scale:Vector3f=&Multiply_Vector(
     A=<0.0785 0.06 0.06>
     B=&MakeVector(
      X=&Multiply_Float(
       A=&FloatFromCurve(
        FloatCurve=DataInterface() {propertyValues {Curve {keys [<at 0 0.5 RCIM_Cubic>, {arriveTangent=10.1608, leaveTangent=10.1608, time=0.225, value=3}, <at 0.275 3>, <at 0.5 1>]}}}
        Scale Curve=&FloatFromCurve(
         FloatCurve=DataInterface() {propertyValues {Curve {keys [
             <at 0.1 1 RCIM_Cubic>
             {arriveTangent=0.5308, interpMode=RCIM_Cubic, leaveTangent=0.5308, time=0.3, value=0.8297}
             {arriveTangent=-0.0508, interpMode=RCIM_Cubic, leaveTangent=-0.0508, time=0.55, value=1.2389}
             {arriveTangent=-0.7962, interpMode=RCIM_Cubic, leaveTangent=-0.7962, time=0.67, value=0.8109}
             <at 0.85 1>
            ]}}}
         CurveIndex=&Add_Float(A=$Emitter.NormalizedLoopAge, B=&MaskFloatBySpawnGroup(Masked Float=0, Passthrough Float=0.1))
        )
       )
       B=1
      )
      Y=1
      Z=1
     )
    )
   )
   &FN_MeshRenderer_v2(
    Color:LinearColor=<rgba 8 0.8526 0 0.35>
    Create COlor Attribute=true
    Create Orientation Attribute=true
    Create Scale Attribute=true
    Orientation:Quat4f=&MultiplyQuaternion(
     Quaternion A=$Particles.MeshOrientation
     Quaternion B=&FindQuatBetweenTwoVectors(From Vector=<-1 0 0>, To Vector=<0 0 1>)
    )
    Scale:Vector3f=&Multiply_Vector(
     A=<0.0985 0.1 0.1>
     B=&MakeVector(
      X=&Multiply_Float(
       A=&FloatFromCurve(
        FloatCurve=DataInterface() {propertyValues {Curve {keys [<at 0 0.5 RCIM_Cubic>, {arriveTangent=10.1608, leaveTangent=10.1608, time=0.225, value=3}, <at 0.275 3>, <at 0.5 1>]}}}
        Scale Curve=&FloatFromCurve(
         FloatCurve=DataInterface() {propertyValues {Curve {keys [
             <at 0.1 1 RCIM_Cubic>
             {arriveTangent=0.5308, interpMode=RCIM_Cubic, leaveTangent=0.5308, time=0.3, value=0.8297}
             {arriveTangent=-0.0508, interpMode=RCIM_Cubic, leaveTangent=-0.0508, time=0.55, value=1.2389}
             {arriveTangent=-0.7962, interpMode=RCIM_Cubic, leaveTangent=-0.7962, time=0.67, value=0.8109}
             <at 0.85 1>
            ]}}}
         CurveIndex=&Add_Float(A=$Emitter.NormalizedLoopAge, B=&MaskFloatBySpawnGroup(Masked Float=0, Passthrough Float=0.1))
        )
       )
       B=1
      )
      Y=1
      Z=1
     )
    )
   ) {module=FN_MeshRenderer_v2001}
   &FN_MeshRenderer_v2(
    Color:LinearColor=<rgba 12 0.6147 0 1>
    Create COlor Attribute=true
    Create Scale Attribute=true
    Scale:Vector3f=&Multiply_Vector(
     A=<0.09 0.085 0.085>
     B=&MakeVector(
      X=&Multiply_Float(
       A=&FloatFromCurve(
        FloatCurve=DataInterface() {propertyValues {Curve {keys [<at 0 0.5 RCIM_Cubic>, {arriveTangent=10.1608, leaveTangent=10.1608, time=0.225, value=3}, <at 0.275 3>, <at 0.5 1>]}}}
        Scale Curve=&FloatFromCurve(
         FloatCurve=DataInterface() {propertyValues {Curve {keys [
             <at 0.1 1 RCIM_Cubic>
             {arriveTangent=0.5308, interpMode=RCIM_Cubic, leaveTangent=0.5308, time=0.3, value=0.8297}
             {arriveTangent=-0.0508, interpMode=RCIM_Cubic, leaveTangent=-0.0508, time=0.55, value=1.2389}
             {arriveTangent=-0.7962, interpMode=RCIM_Cubic, leaveTangent=-0.7962, time=0.67, value=0.8109}
             <at 0.85 1>
            ]}}}
         CurveIndex=&Add_Float(A=$Emitter.NormalizedLoopAge, B=&MaskFloatBySpawnGroup(Masked Float=0, Passthrough Float=0.1))
        )
       )
       B=1
      )
      Y=1
      Z=1
     )
    )
   ) {module=FN_MeshRenderer_v2002}
   &FN_Light_v3(
    Color=<rgba 1.25 0.427 0 1>
    Create Color Attribute=true
    Create Exponent Attribute=true
    Create Position Attribute=true
    Create Radius Attribute=true
    Exponent=4
    Position:Vector3f=&Add_Vector(
     A=Linked() {linkedVariable {name=Particles.Position, type=NiagaraPosition}}
     B=&RotateVectorByQuaternion(Vector To Rotate=<0 -2 -5>, Quaternion=$Output.SkeletalMeshLocation.SampledRotation)
    )
    Radius=25
    Scale Color:NiagaraFloat=&FloatFromCurve(
     FloatCurve=DataInterface() {propertyValues {Curve {keys [
         <at 0.1 1>
         {arriveTangent=0.7239, interpMode=RCIM_Cubic, leaveTangent=0.7239, time=0.25, value=1.4302}
         {arriveTangent=0.2419, interpMode=RCIM_Cubic, leaveTangent=0.2419, time=0.5, value=1.2896}
         {arriveTangent=-0.5791, interpMode=RCIM_Cubic, leaveTangent=-0.5791, time=0.75, value=1.5512}
         <at 1 1>
        ]}}}
     CurveIndex=$Emitter.NormalizedLoopAge
    )
   )
   &ScaleColor(Scale Alpha=$System.VehicleAlpha, ScaleA=true, ScaleRGB=true)
  ]
  rendererClasses [&NiagaraMeshRendererProperties, &NiagaraLightRendererProperties]
  renderers [
   {rendererClass=&NiagaraMeshRendererProperties, rendererIndex=0}
   {rendererClass=&NiagaraMeshRendererProperties, rendererIndex=1}
   {rendererClass=&NiagaraMeshRendererProperties, rendererIndex=2}
   {rendererClass=&NiagaraLightRendererProperties, rendererIndex=3}
  ]
 }
 JetPackSmoke {
  emitterSpawnScript [&FN_PreserveSkydiveUserVariables() {module=PreserveUserVariables}]
  emitterUpdateScript [
   &EmitterState(Delay First Loop Only=true, Life Cycle Mode=::NewEnumerator1, Loop Delay=0.2, Loop Duration=1, UseLoopDelay=true)
   &SpawnRate(SpawnRate=20, Use Spawn Probability=false)
  ]
  particleSpawnScript [
   &FN_SetSpawnGroup_v2(Number of Spawn Groups=2)
   &InitializeParticle(
    Color=<rgba 0.544 0.6061 0.6771 0.225>
    Lifetime=&UniformRangedFloat(Minimum=0.35, Maximum=0.4)
    Module.Write Color=true
    Module.Write Lifetime=true
    Module.Write Mass=false
    Module.Write Material Random=true
    Module.Write Position=false
    Module.Write Scale=false
    Module.Write SpriteRotation=true
    Module.Write SpriteSize=true
    Sprite Rotation=&UniformRangedFloat(Minimum=-360, Maximum=360)
    Sprite Size=&Vector2DFromFloat(Value=&UniformRangedFloat(Minimum=4, Maximum=6))
   )
   &CameraOffset(Camera Offset Amount=2)
  ]
  particleUpdateScript [
   &ParticleState()
   &SkeletalMeshLocation(
    Bone / Socket Index=$Particles.SpawnGroup
    Inherit Velocity=false
    MaskSpawnGroup=false
    Mesh Sampling Type=::NewEnumerator5
    Sampled Orientation Coordinate Space=::World
    Sampled Position Coordinate Space=::World
    Sampled Velocity Coordinate Space:ENiagaraCoordinateSpace=::World
    Skeletal Mesh=$User.SkeletalMesh
    Socket Sampling Mode=::NewEnumerator5
   )
   /SetVariables(Particles.Position=&Add_Vector(
     A=Linked() {linkedVariable {name=Particles.Position, type=NiagaraPosition}}
     B=&RotateVectorByQuaternion(
      Vector To Rotate=&Multiply_VectorByFloat(
       Vector=&MakeVector(
        X=0
        Y=&Multiply_Float(
         A=-5
         B=&FloatFromCurve(FloatCurve=DataInterface() {propertyValues {Curve {keys [<at 0.25 0 RCIM_Cubic>,<at 1 1>]}}})
        )
        Z=-12
       )
       Float=$Particles.NormalizedAge
      )
      Quaternion=$Output.SkeletalMeshLocation.SampledRotation
     )
    )) {isSetParameters=true, module=SetVariables_#2}
   &ScaleSpriteSize(Scale Factor=&Vector2DFromFloat(Value=&FloatFromCurve(FloatCurve=DataInterface() {propertyValues {Curve {keys [<at 0 0.5 RCIM_Cubic>,<at 0.75 2>]}}})))
   /Niagara/Modules/Update/SubUV/SubUVAnimation(Number Of Frames=64, SubUV Lookup Index=&Multiply_Float(A=$Particles.NormalizedAge, B=0.75))
   &ScaleColor(
    Scale Alpha=&FloatFromCurve(FloatCurve=DataInterface() {propertyValues {Curve {keys [<at 0 0 RCIM_Cubic>,<at 0.1 1 RCIM_Cubic>,<at 0.75 0>]}}})
    ScaleA=true
    ScaleRGB=false
   )
   &DynamicMaterialParameters(
    Index 0 Param 1=&FloatFromCurve(FloatCurve=DataInterface() {propertyValues {Curve {keys [<at 0 1>,<at 1 -1>]}}})
    Index 0 Param 4=1
    Module.Param0WriteEnabled=true
    Module.Param1WriteEnabled=false
    Module.Param2WriteEnabled=false
    Module.Param3WriteEnabled=true
   )
   &ScaleColor(Scale Alpha=$System.VehicleAlpha, ScaleA=true, ScaleRGB=true) {module=ScaleColor001}
  ]
  rendererClasses [&NiagaraSpriteRendererProperties]
  renderers [{rendererClass=&NiagaraSpriteRendererProperties}]
 }
 JetPackSmokeInitialBurst {
  emitterSpawnScript [&FN_PreserveSkydiveUserVariables() {module=PreserveUserVariables}]
  emitterUpdateScript [
   &EmitterState(Life Cycle Mode=::NewEnumerator1, Loop Behavior=::NewEnumerator1, Loop Duration=0.5, UseLoopDelay=false)
   &SpawnRate(SpawnRate=60, Use Spawn Probability=false)
  ]
  particleSpawnScript [
   &FN_SetSpawnGroup_v2(Number of Spawn Groups=2)
   &InitializeParticle(
    Color=<rgba 0.3934 0.4383 0.4896 0.225>
    Lifetime=&UniformRangedFloat(Minimum=0.35, Maximum=0.4)
    Module.Write Color=true
    Module.Write Lifetime=true
    Module.Write Mass=false
    Module.Write Material Random=true
    Module.Write Position=false
    Module.Write Scale=false
    Module.Write SpriteRotation=true
    Module.Write SpriteSize=true
    Sprite Rotation=&UniformRangedFloat(Minimum=-360, Maximum=360)
    Sprite Size=&Vector2DFromFloat(Value=&UniformRangedFloat(Minimum=4, Maximum=6))
   )
   &CameraOffset(Camera Offset Amount=2)
   /SetVariables(Particles.Offset:NiagaraFloat=&UniformRangedFloat(Minimum=-20, Maximum=20)) {isSetParameters=true, module=SetVariables_#3}
  ]
  particleUpdateScript [
   &ParticleState()
   &SkeletalMeshLocation(
    Bone / Socket Index=$Particles.SpawnGroup
    Inherit Velocity=false
    MaskSpawnGroup=false
    Mesh Sampling Type=::NewEnumerator5
    Sampled Orientation Coordinate Space=::World
    Sampled Position Coordinate Space=::World
    Sampled Velocity Coordinate Space:ENiagaraCoordinateSpace=::World
    Skeletal Mesh=$User.SkeletalMesh
    Socket Sampling Mode=::NewEnumerator5
   )
   /SetVariables(Particles.Position=&Add_Vector(
     A=Linked() {linkedVariable {name=Particles.Position, type=NiagaraPosition}}
     B=&RotateVectorByQuaternion(
      Vector To Rotate=&Multiply_VectorByFloat(
       Vector=&MakeVector(
        X=&Multiply_Float(A=$Particles.Initial.Offset, B=$Particles.NormalizedAge)
        Y=&Multiply_Float(A=-40, B=$Particles.NormalizedAge)
        Z=-70
       )
       Float=$Particles.NormalizedAge
      )
      Quaternion=$Output.SkeletalMeshLocation.SampledRotation
     )
    )) {isSetParameters=true, module=SetVariables_#2}
   &ScaleSpriteSize(Scale Factor=&Vector2DFromFloat(Value=&FloatFromCurve(FloatCurve=DataInterface() {propertyValues {Curve {keys [<at 0 0.5 RCIM_Cubic>,<at 0.75 6>]}}})))
   &SubUVAnimation(Number Of Frames=64, SubUV Lookup Index=&Multiply_Float(A=$Particles.NormalizedAge, B=0.75))
   &ScaleColor(
    Scale Alpha=&Multiply_Float(
     A=&FloatFromCurve(FloatCurve=DataInterface() {propertyValues {Curve {keys [<at 0 0 RCIM_Cubic>,<at 0.1 1 RCIM_Cubic>,<at 1 0>]}}})
     B=&FloatFromCurve(
      FloatCurve=DataInterface() {propertyValues {Curve {keys [
          <at 0 0 RCIM_Cubic>
          {arriveTangent=2, leaveTangent=2, time=0.05, value=1}
          {interpMode=RCIM_Cubic, leaveTangent=-1.0526, time=0.5, value=1}
          <at 0.9 0>
         ]}}}
      CurveIndex=$System.NormalizedLoopAge
     )
    )
    ScaleA=true
    ScaleRGB=false
   )
   &DynamicMaterialParameters(
    Index 0 Param 1=&FloatFromCurve(FloatCurve=DataInterface() {propertyValues {Curve {keys [<at 0 1>,<at 1 -1>]}}})
    Index 0 Param 4=1
    Module.Param0WriteEnabled=true
    Module.Param1WriteEnabled=false
    Module.Param2WriteEnabled=false
    Module.Param3WriteEnabled=true
   )
   &ScaleColor(Scale Alpha=$System.VehicleAlpha, ScaleA=true, ScaleRGB=true) {module=ScaleColor001}
  ]
  rendererClasses [&NiagaraSpriteRendererProperties]
  renderers [{rendererClass=&NiagaraSpriteRendererProperties}]
 }
 SmokeInitialBurst {
  emitterSpawnScript [&FN_PreserveSkydiveUserVariables() {module=PreserveUserVariables}]
  emitterUpdateScript [
   &EmitterState(Life Cycle Mode=::NewEnumerator1, Loop Behavior=::NewEnumerator1, Loop Duration=0.5, UseLoopDelay=false)
   &SpawnRate(SpawnRate=10, Use Spawn Probability=false)
  ]
  particleSpawnScript [
   &FN_SetSpawnGroup_v2(Number of Spawn Groups=2)
   &InitializeParticle(
    Color=<rgba 0.1314 0.1406 0.151 0.25>
    Lifetime=&UniformRangedFloat(Minimum=0.5, Maximum=0.6)
    Module.Write Color=true
    Module.Write Lifetime=true
    Module.Write Mass=false
    Module.Write Material Random=true
    Module.Write Position=false
    Module.Write Scale=false
    Module.Write SpriteRotation=true
    Module.Write SpriteSize=true
    Sprite Rotation=&UniformRangedFloat(Minimum=-360, Maximum=360)
    Sprite Size=&Vector2DFromFloat(Value=&UniformRangedFloat(Minimum=7, Maximum=8))
   )
   &CameraOffset(Camera Offset Amount=2)
   /SetVariables(Particles.Offset:NiagaraFloat=&UniformRangedFloat(Minimum=-40, Maximum=40)) {isSetParameters=true, module=SetVariables_#3}
  ]
  particleUpdateScript [
   &ParticleState()
   &SkeletalMeshLocation(
    Bone / Socket Index=$Particles.SpawnGroup
    Inherit Velocity=false
    MaskSpawnGroup=false
    Mesh Sampling Type=::NewEnumerator5
    Sampled Orientation Coordinate Space=::World
    Sampled Position Coordinate Space=::World
    Sampled Velocity Coordinate Space:ENiagaraCoordinateSpace=::World
    Skeletal Mesh=$User.SkeletalMesh
    Socket Sampling Mode=::NewEnumerator5
   )
   /SetVariables(Particles.Position=&Add_Vector(
     A=Linked() {linkedVariable {name=Particles.Position, type=NiagaraPosition}}
     B=&RotateVectorByQuaternion(
      Vector To Rotate=&Multiply_VectorByFloat(
       Vector=&MakeVector(
        X=&Multiply_Float(A=$Particles.Initial.Offset, B=$Particles.NormalizedAge)
        Y=&Multiply_Float(A=-60, B=$Particles.NormalizedAge)
        Z=-100
       )
       Float=$Particles.NormalizedAge
      )
      Quaternion=$Output.SkeletalMeshLocation.SampledRotation
     )
    )) {isSetParameters=true, module=SetVariables_#2}
   &ScaleSpriteSize(Scale Factor=&Vector2DFromFloat(Value=&FloatFromCurve(FloatCurve=DataInterface() {propertyValues {Curve {keys [<at 0 0.5 RCIM_Cubic>,<at 0.75 5>]}}})))
   &SubUVAnimation(Number Of Frames=64, SubUV Lookup Index=&Multiply_Float(A=$Particles.NormalizedAge, B=0.75))
   &ScaleColor(
    Scale Alpha=&Multiply_Float(
     A=&FloatFromCurve(FloatCurve=DataInterface() {propertyValues {Curve {keys [<at 0 0 RCIM_Cubic>,<at 0.05 1 RCIM_Cubic>,<at 1 0>]}}})
     B=&FloatFromCurve(
      FloatCurve=DataInterface() {propertyValues {Curve {keys [
          <at 0 0 RCIM_Cubic>
          {arriveTangent=2, leaveTangent=2, time=0.05, value=1}
          {interpMode=RCIM_Cubic, leaveTangent=-1.0526, time=0.5, value=1}
          <at 0.9 0>
         ]}}}
      CurveIndex=$System.NormalizedLoopAge
     )
    )
    ScaleA=false
    ScaleRGB=false
   )
   &DynamicMaterialParameters(
    Index 0 Param 1=&FloatFromCurve(FloatCurve=DataInterface() {propertyValues {Curve {keys [<at 0.7 0.5 RCIM_Cubic>,<at 1 -1>]}}})
    Index 0 Param 4=&FloatFromCurve(FloatCurve=DataInterface() {propertyValues {Curve {keys [<at 0.6 1>,<at 1 0>]}}})
    Module.Param0WriteEnabled=true
    Module.Param1WriteEnabled=true
    Module.Param2WriteEnabled=true
    Module.Param3WriteEnabled=true
   )
   &ScaleColor(Scale Alpha=$System.VehicleAlpha, ScaleA=true, ScaleRGB=true) {module=ScaleColor001}
  ]
  rendererClasses [&NiagaraSpriteRendererProperties]
  renderers [{rendererClass=&NiagaraSpriteRendererProperties}]
 }
}
summary {
 systemName=NS_Pet_HighTower_Radish_Jets
 emitters [
  {emitterName=JetPackFlames, rendererClasses [&NiagaraMeshRendererProperties, &NiagaraLightRendererProperties]}
  {emitterName=JetPackSmoke, rendererClasses [&NiagaraSpriteRendererProperties]}
  {emitterName=JetPackSmokeInitialBurst, rendererClasses [&NiagaraSpriteRendererProperties]}
  {emitterName=SmokeInitialBurst, rendererClasses [&NiagaraSpriteRendererProperties]}
 ]
 userVariables [
  User.bIsFrontEndPreview {type=NiagaraBool}
  User.bisInVehicle {type=NiagaraBool}
  User.PlayerSpeed {type=NiagaraFloat, defaultValue=0, description=""}
  User.SkeletalMesh {
   type=NiagaraDataInterfaceSkeletalMesh
   defaultValue {
    struct=NiagaraExt_VariableValue_DataInterface
    value {dataInterface=&self:NiagaraDataInterfaceSkeletalMesh_0}
   }
  }
 ]
}
